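# Looping ambient SoundCue for the Chaos Emerald spot sound at the GenoBase POI.
# A single looping wave player feeds the root directly, playing the Chaos Emerald spot wave.
# Spatialization and falloff come from the GenoBase emerald attenuation asset, routed through the Athena ambient props sound class.
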SoundCue /Hera/Audio/POI/GenoBase/IP_SC_Spot_GenoBase_ChaosEmerald_Loop
root=WavePlayer_1
nodes [{id=WavePlayer_1, properties {bLooping=true, soundWaveAssetPtr=&/IP_Spot_GenoBase_Chaos_Emerald_01}}]
properties {
 attenuationSettings=&/ATT_GenoBase_Emerald
 maxDistance=3800
 soundClassObject=&Fort_Ambient_Props_Athena
}
