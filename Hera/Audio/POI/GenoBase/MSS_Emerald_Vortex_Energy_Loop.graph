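# Stereo MetaSound source that plays the Emerald Vortex energy loop as ambient prop audio, apparently for a GenoBase POI.
# On play, it triggers one stereo Wave Player on the Sound input, routing left and right straight to the outputs.
# The source is flagged looping with attenuation and lowered pitch, though the player's Loop default holds a builder-result string rather than a boolean.
# A second wave input, Input, is unconnected.
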MetaSoundSource /Hera/Audio/POI/GenoBase/MSS_Emerald_Vortex_Energy_Loop
inputs [
 Input {type=WaveAsset}
 Sound {type=WaveAsset}
 UE.Source.OnPlay {type=Trigger}
]
interfaces [UE.OutputFormat.Stereo:1.0, UE.Source:1.1]
outputs [
 "UE.OutputFormat.Stereo.Audio:0" {type=Audio}
 "UE.OutputFormat.Stereo.Audio:1" {type=Audio}
]
pages [Default {
  input_defaults {Input=null, Sound=&/Emerald_Vortex_Loop_Mixdown_01, UE.Source.OnPlay=null}
  nodes [WavePlayer_0 = UE.Wave Player:Stereo(Wave Asset=^Sound, Play=^UE.Source.OnPlay) {also {defaults {Loop="<MetaSoundBuilderResult.SUCCEEDED: 0>"}, version [1,1]}}]
  links [{dst="@UE.OutputFormat.Stereo.Audio:0", src=WavePlayer_0.Out Left}, {dst="@UE.OutputFormat.Stereo.Audio:1", src=WavePlayer_0.Out Right}]
 }]
properties {
 assetImportData=&self:AssetImportData
 attenuationSettings=&/ATT_MSS_Emerald_Vortex_Energy_Loop
 bLooping=true
 duration=10000
 maxDistance=2050
 numChannels=2
 outputFormat=Stereo
 pitch=0.85
 sampleRate=48000
 soundClassObject=&Fort_Ambient_Props_Athena
 volume=0.9
}
