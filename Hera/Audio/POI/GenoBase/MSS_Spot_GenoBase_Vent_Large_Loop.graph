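# Mono MetaSound source for a looping large-vent ambient spot sound at the GenoBase POI, with its own attenuation.
# On Play picks a random wave from five SecretFacility vent loops and starts the Wave Player.
# The chosen wave's duration caps a Random Float that feeds Start Time, intended to vary the loop's entry point.
# That Random Float has no trigger wired to it.
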MetaSoundSource /Hera/Audio/POI/GenoBase/MSS_Spot_GenoBase_Vent_Large_Loop
inputs [
 Sounds {type=WaveAsset:Array}
 UE.Source.OnPlay {type=Trigger}
]
interfaces [UE.OutputFormat.Mono:1.0, UE.Source:1.1]
outputs ["UE.OutputFormat.Mono.Audio:0" {type=Audio}]
pages [Default {
  input_defaults {
   UE.Source.OnPlay=null
   Sounds [
    /Hera/Audio/POI/SecretFacility/Spots/Spot_SecretFacility_Vent_Large_Loop_01
    /Hera/Audio/POI/SecretFacility/Spots/Spot_SecretFacility_Vent_Large_Loop_02
    /Hera/Audio/POI/SecretFacility/Spots/Spot_SecretFacility_Vent_Large_Loop_03
    /Hera/Audio/POI/SecretFacility/Spots/Spot_SecretFacility_Vent_Large_Loop_04
    /Hera/Audio/POI/SecretFacility/Spots/Spot_SecretFacility_Vent_Large_Loop_05
   ]
  }
  nodes [
   ArrayRandomGet_0 = Array.Random Get:WaveAsset:Array(In Array=^Sounds, Next=^UE.Source.OnPlay)
   GetWaveDuration_0 = "UE.Get Wave Duration: "(Wave=@ArrayRandomGet_0.Value)
   n1 = UE.ConversionTimeToFloat(In=@GetWaveDuration_0.Duration)
   WavePlayer_0 = UE.Wave Player:Mono(
    Play=@ArrayRandomGet_0.On Next
    Wave Asset=@ArrayRandomGet_0.Value
    Start Time=UE.ConversionFloatToTime(In=UE.RandomFloat(Max=@n1.Out).Value).Out
   ) {also {defaults {Loop="<MetaSoundBuilderResult.SUCCEEDED: 0>"}, version [1,1]}}
  ]
  links [{dst="@UE.OutputFormat.Mono.Audio:0", src=WavePlayer_0.Out Mono}]
 }]
properties {
 assetImportData=&self:AssetImportData
 attenuationSettings=&/ATT_GenoBase_ShellVent_Large
 bLooping=true
 duration=10000
 maxDistance=2600
 sampleRate=48000
 soundClassObject=&Fort_Ambient_Outdoors_Athena
 volume=2
}
