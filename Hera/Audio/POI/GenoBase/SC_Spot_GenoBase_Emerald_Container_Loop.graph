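# Sound cue for a looping ambient spot sound on an emerald container at the GenoBase point of interest.
# A single looping wave player plays the container wave directly as the root, with no randomization or modulation.
# The large-prop ambient emitter attenuation and the Athena ambient props sound class govern its falloff and mixing.
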SoundCue /Hera/Audio/POI/GenoBase/SC_Spot_GenoBase_Emerald_Container_Loop
root=WavePlayer_0
nodes [{id=WavePlayer_0, properties {bLooping=true, soundWaveAssetPtr=&/Spot_GenoBase_Emerald_Container_01}}]
properties {
 attenuationSettings=/Game/Sounds/Attenuation/Fort_Amb_Emitter_Prop_Large_Att
 maxDistance=2000
 soundClassObject=&Fort_Ambient_Props_Athena
}
